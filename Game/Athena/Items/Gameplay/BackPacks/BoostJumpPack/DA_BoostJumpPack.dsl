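; Boost Jump Pack deco tool item that answers a trigger press with a usage error rather than performing an action.
; The error reason varies for gamepad or Nintendo Switch versus other input, and a vehicle-specific reason overrides it when the instigator is in a vehicle.
; When the item becomes visible, it forces the pawn's backpack mesh visible, using a local-control check and a short delay.
; An IsOnGround helper reports whether the instigator is not falling.
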(event EventOnWeaponVisibilityChanged
  (params (bVisible :type Boolean) (bSetForLocalControllerOnly :type Boolean))
  (bind _self self)
  (bind _returnvalue (Class|FortKismetLibrary|GetPawnfromActor _self "/Script/FortniteGame.FortPlayerPawn"))
  (bind _returnvalue_2 (Customization|GetSkeletalMeshforPartType _returnvalue "Backpack" :!class /Script/FortniteGame.FortPlayerPawn))
  (if bVisible
    (Utilities|IsValid _returnvalue
      (:Is%20Valid
        (if (Pawn|IsLocallyControlled _returnvalue)
          (Rendering|SetVisibility _returnvalue_2 true false)
          (Utilities|FlowControl|Delay 0.2)
          (Rendering|SetVisibility _returnvalue_2 true false))))))

(event DecoTool|EventOnPressTrigger
  (params (FortDecoHelper :type Fort%20Deco%20Helper%20Object%20Reference))
  (CallFunction|RaiseUsageError (select (or (Utilities|Enum|Equal(Enum) (CommonInputSubsystem|GetCurrentInputType (BlueprintContext|GetCommonInputSubsystem)) "Gamepad") (TouchPlatform|IsNintendoSwitch)) (Variables|Default|GetErrorReasonGamepad) (Variables|Default|GetErrorReason))))

(fn IsOnGround
  (bind _asfort_player_pawn (Utilities|Casting|CastToFortPlayerPawn (Game|GetInstigator))
    (:then
      (return (not (AI|Components|NavMovement|IsFalling (Variables|/Script/Engine.Character|GetCharacterMovement _asfort_player_pawn)))))
    (:CastFailed
      (return true))))

(fn RaiseUsageError
  (params (FailedReason :type Gameplay%20Tag%20Container%20Structure%20%28by%20ref%29))
  (bind _returnvalue (Game|GetInstigator))
  (bind _asfort_player_pawn (Utilities|Casting|CastToFortPlayerPawn _returnvalue))
  (Utilities|IsinVehicle _returnvalue
    (:true
      (FailedAction|UnabletoPerformAction _asfort_player_pawn false (LocalVariables|GetVehicleErrorReason)))
    (:false
      (FailedAction|UnabletoPerformAction _asfort_player_pawn false (Variables|GetFailedReason)))))

(fn ConstructionScript)
